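; Frontend mannequin Blueprint for the BuffCat character that plays a transition effect and sound when its cosmetics change.
; Part-changed and variant-changed notifications run identical logic, ignoring which part or variant changed.
; Each attaches the full-body paper-doll transition FX to the skeletal mesh at spine_05 with an offset.
; It also plays the faction transition UI cue at the mesh's world location.
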(event MannequinNotifications|EventOnPartChanged
  (params (OldPart :type Custom%20Character%20Part%20Object%20Reference) (NewPart :type Custom%20Character%20Part%20Object%20Reference) (PartType :type EFortCustomPartType%20Enum :default "Head"))
  (bind _skeletalmeshcomponent (Variables|SkeletalMeshActor|GetSkeletalMeshComponent))
  (Effects|SpawnFXSystemAttached "/BRCosmetics/Effects/Fort_Effects/UI/Transition/NS_Frontend_PaperDoll_Transition_FullBody.NS_Frontend_PaperDoll_Transition_FullBody" _skeletalmeshcomponent "spine_05" "-50.000000, 0, 0")
  (Audio|PlaySoundatLocation "/BRCosmetics/Sounds/Other/Fort_UI_Sounds/S12_Frontend/UI_Faction_Transition_Cue.UI_Faction_Transition_Cue" (Transformation|GetWorldLocation _skeletalmeshcomponent)))

(event MannequinNotifications|EventOnVariantChanged
  (params (VariantChannel :type Gameplay%20Tag%20Structure) (OldVariantTag :type Gameplay%20Tag%20Structure) (NewVariantTag :type Gameplay%20Tag%20Structure))
  (bind _skeletalmeshcomponent (Variables|SkeletalMeshActor|GetSkeletalMeshComponent))
  (Effects|SpawnFXSystemAttached "/BRCosmetics/Effects/Fort_Effects/UI/Transition/NS_Frontend_PaperDoll_Transition_FullBody.NS_Frontend_PaperDoll_Transition_FullBody" _skeletalmeshcomponent "spine_05" "-50.000000, 0, 0")
  (Audio|PlaySoundatLocation "/BRCosmetics/Sounds/Other/Fort_UI_Sounds/S12_Frontend/UI_Faction_Transition_Cue.UI_Faction_Transition_Cue" (Transformation|GetWorldLocation _skeletalmeshcomponent)))

(fn ConstructionScript)
