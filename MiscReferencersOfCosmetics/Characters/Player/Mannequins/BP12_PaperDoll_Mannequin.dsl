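; Frontend paper doll mannequin Blueprint that plays transition effects and sounds when a character's cosmetic parts or variants change.
; Part changes switch on part type to attach a head or chest Niagara effect to the skeletal mesh, then play a transition cue.
; Variant changes switch on channel tag to pick head, chest, arms, legs or feet effects, with arms reusing a Build-A-Brella canopy effect.
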(event MannequinNotifications|EventOnPartChanged
  (params (OldPart :type Custom%20Character%20Part%20Object%20Reference) (NewPart :type Custom%20Character%20Part%20Object%20Reference) (PartType :type EFortCustomPartType%20Enum :default "Head"))
  ;[0.09255 0.2702 1] Spawn FX on Part Swap
  (switch Utilities|FlowControl|Switch|SwitchonEFortCustomPartType PartType
    (:Head
      ;[0 0.3613 0.402] HeadFX
      (Effects|SpawnFXSystemAttached "/BRCosmetics/Effects/Fort_Effects/UI/Transition/NS_Frontend_PaperDoll_Transition_Head.NS_Frontend_PaperDoll_Transition_Head" (Variables|SkeletalMeshActor|GetSkeletalMeshComponent) "head" "-50.000000, 0, 0" :!label _label))
    (:Body
      ;[0 0.3613 0.402] ChestFX
      (Effects|SpawnFXSystemAttached "/BRCosmetics/Effects/Fort_Effects/UI/Transition/NS_Frontend_PaperDoll_Transition.NS_Frontend_PaperDoll_Transition" (Variables|SkeletalMeshActor|GetSkeletalMeshComponent) "spine_05" "-50.000000, 0, 0" :!label _label_2))
    (:Hat
      (goto _label))
    (:Backpack
      (goto _label_2))
    (:MiscOrTail
      (goto _label_2))
    (:Face
      (goto _label)))
  (Audio|PlaySoundatLocation "/BRCosmetics/Sounds/Other/Fort_UI_Sounds/S12_Frontend/UI_PaperDoll_Part_Transition_Cue.UI_PaperDoll_Part_Transition_Cue" (Transformation|GetWorldLocation (Variables|SkeletalMeshActor|GetSkeletalMeshComponent))))

(event MannequinNotifications|EventOnVariantChanged
  (params (VariantChannel :type Gameplay%20Tag%20Structure) (OldVariantTag :type Gameplay%20Tag%20Structure) (NewVariantTag :type Gameplay%20Tag%20Structure))
  ;[0.09255 0.2702 1] Spawn FX on Part Swap
  (switch Utilities|FlowControl|Switch|SwitchonGameplayTag VariantChannel
    (:Cosmetics.Variant.Channel.Emissive
      _)
    (:Cosmetics.Variant.Channel.ClothingColor
      _)
    (:Cosmetics.Variant.Channel.Particle
      ;[0 0.3613 0.402] FootFX
      (Effects|SpawnFXSystemAttached "/BRCosmetics/Effects/Fort_Effects/UI/Transition/NS_Frontend_PaperDoll_Transition_Feet.NS_Frontend_PaperDoll_Transition_Feet" (Variables|SkeletalMeshActor|GetSkeletalMeshComponent) "spine_05" "-50.000000, 0, 0")
      (Audio|PlaySoundatLocation "/BRCosmetics/Sounds/Other/Fort_UI_Sounds/S12_Frontend/UI_PaperDoll_Part_Transition_Cue.UI_PaperDoll_Part_Transition_Cue" (Transformation|GetWorldLocation (Variables|SkeletalMeshActor|GetSkeletalMeshComponent)) :!label _label))
    (:Cosmetics.Variant.Channel.Material
      (Utilities|FlowControl|Sequence
        (:then_0
          ;[0 0.3613 0.402] ChestFX
          (Effects|SpawnFXSystemAttached "/BRCosmetics/Effects/Fort_Effects/UI/Transition/NS_Frontend_PaperDoll_Transition.NS_Frontend_PaperDoll_Transition" (Variables|SkeletalMeshActor|GetSkeletalMeshComponent) "spine_05" "-50.000000, 0, 0"))
        (:then_1
          ;[0 0.3613 0.402] ArmsFX
          (Effects|SpawnFXSystemAttached "/BRCosmetics/Gadgets/Assets/VinderTech_GliderChute/Glider_Umbrella_Build_A_Brella/FX/NS_Frontend_Build_A_Brella_Transition_Canopy.NS_Frontend_Build_A_Brella_Transition_Canopy" (Variables|SkeletalMeshActor|GetSkeletalMeshComponent) "spine_05" "-50.000000, 0, 0" :!label _label_3)))
      (goto _label))
    (:Cosmetics.Variant.Channel.JerseyColor
      ;[0 0.3613 0.402] LegsFX
      (Effects|SpawnFXSystemAttached "/BRCosmetics/Effects/Fort_Effects/UI/Transition/NS_Frontend_PaperDoll_Transition_LowerBody.NS_Frontend_PaperDoll_Transition_LowerBody" (Variables|SkeletalMeshActor|GetSkeletalMeshComponent) "spine_05" "-50.000000, 0, 0")
      (goto _label))
    (:Cosmetics.Variant.Channel.Mesh
      ;[0 0.3613 0.402] HeadFX
      (Effects|SpawnFXSystemAttached "/BRCosmetics/Effects/Fort_Effects/UI/Transition/NS_Frontend_PaperDoll_Transition_Head.NS_Frontend_PaperDoll_Transition_Head" (Variables|SkeletalMeshActor|GetSkeletalMeshComponent) "head" "-50.000000, 0, 0" :!label _label_2)
      (goto _label))
    (:Cosmetics.Variant.Channel.Pattern
      (goto _label_2)))
  (goto _label_3))

(fn ConstructionScript)
