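; Black Knight player mannequin Blueprint that plays transition feedback when its cosmetic parts or variants change.
; The part-changed and variant-changed events each spawn the frontend paper-doll full-body Niagara effect, attached at an offset to the skeletal mesh's head socket.
; Both also play the part-transition sound cue at the mesh's world location.
; The two handlers are identical and ignore their parameters; the construction script is empty.
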(event MannequinNotifications|EventOnPartChanged
  (params (OldPart :type Custom%20Character%20Part%20Object%20Reference) (NewPart :type Custom%20Character%20Part%20Object%20Reference) (PartType :type EFortCustomPartType%20Enum :default "Head"))
  (Effects|SpawnFXSystemAttached "/BRCosmetics/Effects/Fort_Effects/UI/Transition/NS_Frontend_PaperDoll_Transition_FullBody.NS_Frontend_PaperDoll_Transition_FullBody" (Variables|SkeletalMeshActor|GetSkeletalMeshComponent) "head" "-50.000000, 0, 0")
  (Audio|PlaySoundatLocation "/BRCosmetics/Sounds/Other/Fort_UI_Sounds/S12_Frontend/UI_PaperDoll_Part_Transition_Cue.UI_PaperDoll_Part_Transition_Cue" (Transformation|GetWorldLocation (Variables|SkeletalMeshActor|GetSkeletalMeshComponent))))

(event MannequinNotifications|EventOnVariantChanged
  (params (VariantChannel :type Gameplay%20Tag%20Structure) (OldVariantTag :type Gameplay%20Tag%20Structure) (NewVariantTag :type Gameplay%20Tag%20Structure))
  (Effects|SpawnFXSystemAttached "/BRCosmetics/Effects/Fort_Effects/UI/Transition/NS_Frontend_PaperDoll_Transition_FullBody.NS_Frontend_PaperDoll_Transition_FullBody" (Variables|SkeletalMeshActor|GetSkeletalMeshComponent) "head" "-50.000000, 0, 0")
  (Audio|PlaySoundatLocation "/BRCosmetics/Sounds/Other/Fort_UI_Sounds/S12_Frontend/UI_PaperDoll_Part_Transition_Cue.UI_PaperDoll_Part_Transition_Cue" (Transformation|GetWorldLocation (Variables|SkeletalMeshActor|GetSkeletalMeshComponent))))

(fn ConstructionScript)
