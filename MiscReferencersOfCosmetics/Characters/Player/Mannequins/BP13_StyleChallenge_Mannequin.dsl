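; Style challenge mannequin Blueprint that adds transition and skin-specific effects when cosmetic parts or variants change.
; Both change events spawn a full-body paper-doll transition effect and play a UI sound.
; Part changes also store the new part, then ProfPupSetup picks Neutral, Good or Bad Niagara for Professor Pup, and RacerZeroSetup attaches Racer Zero smoke emitters.
; Stored effects are deactivated before respawn or on unrelated parts, while Racer Zero's head and face parts leave its smoke untouched.
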(event MannequinNotifications|EventOnPartChanged
  (params (OldPart :type Custom%20Character%20Part%20Object%20Reference) (NewPart :type Custom%20Character%20Part%20Object%20Reference) (PartType :type EFortCustomPartType%20Enum :default "Head"))
  (Variables|Default|SetNewPart NewPart)
  (CallFunction|ProfPupSetup)
  (CallFunction|RacerZeroSetup)
  (Effects|SpawnFXSystemAttached "/BRCosmetics/Effects/Fort_Effects/UI/Transition/NS_Frontend_PaperDoll_Transition_FullBody.NS_Frontend_PaperDoll_Transition_FullBody" (Variables|SkeletalMeshActor|GetSkeletalMeshComponent) "head" "-50.000000, 0, 0")
  (Audio|PlaySoundatLocation "/BRCosmetics/Sounds/Other/Fort_UI_Sounds/S12_Frontend/UI_PaperDoll_Part_Transition_Cue.UI_PaperDoll_Part_Transition_Cue" (Transformation|GetWorldLocation (Variables|SkeletalMeshActor|GetSkeletalMeshComponent))))

(event MannequinNotifications|EventOnVariantChanged
  (params (VariantChannel :type Gameplay%20Tag%20Structure) (OldVariantTag :type Gameplay%20Tag%20Structure) (NewVariantTag :type Gameplay%20Tag%20Structure))
  (Effects|SpawnFXSystemAttached "/BRCosmetics/Effects/Fort_Effects/UI/Transition/NS_Frontend_PaperDoll_Transition_FullBody.NS_Frontend_PaperDoll_Transition_FullBody" (Variables|SkeletalMeshActor|GetSkeletalMeshComponent) "head" "-50.000000, 0, 0")
  (Audio|PlaySoundatLocation "/BRCosmetics/Sounds/Other/Fort_UI_Sounds/S12_Frontend/UI_PaperDoll_Part_Transition_Cue.UI_PaperDoll_Part_Transition_Cue" (Transformation|GetWorldLocation (Variables|SkeletalMeshActor|GetSkeletalMeshComponent))))

(fn ProfPupSetup
  (bind _profpupvfxs (Variables|Default|GetProfPupVFXs))
  (bind _profpupvfxs_2 (Variables|Default|GetProfPupVFXs))
  (bind _profpupvfxs_3 (Variables|Default|GetProfPupVFXs))
  (bind _profpupvfxs_4 (Variables|Default|GetProfPupVFXs))
  (if (Utilities|Equal(Object) (Variables|Default|GetNewPart) "/BRCosmetics/Characters/CharacterParts/FaceAccessories/CP_M_LRG_ProfessorPup.CP_M_LRG_ProfessorPup")
    ;[1 1 1] Neutral
    (Utilities|IsValid _profpupvfxs_3
      (:Is%20Valid
        (Components|Activation|Deactivate _profpupvfxs_3))
      (:Is%20Not%20Valid
        _))
    (bind _returnvalue (Niagara|SpawnSystemAttached "/BRCosmetics/Effects/Fort_Effects/Effects/Characters/Athena_Parts/ProfessorPup/NS_ProfessorPup_Neutral.NS_ProfessorPup_Neutral" (Customization|GetPartSkeletalMeshComponent "Body")))
    (Variables|Default|SetProfPupVFXs _returnvalue)
    (elif (Utilities|Equal(Object) (Variables|Default|GetNewPart) "/BRCosmetics/Characters/CharacterParts/FaceAccessories/CP_M_LRG_ProfessorPup_Good.CP_M_LRG_ProfessorPup_Good")
      ;[1 1 1] Good
      (Utilities|IsValid _profpupvfxs_4
        (:Is%20Valid
          (Components|Activation|Deactivate _profpupvfxs_4))
        (:Is%20Not%20Valid
          _))
      (bind _returnvalue_2 (Niagara|SpawnSystemAttached "/BRCosmetics/Effects/Fort_Effects/Effects/Characters/Athena_Parts/ProfessorPup/NS_ProfessorPup_Good.NS_ProfessorPup_Good" (Customization|GetPartSkeletalMeshComponent "Body")))
      (Variables|Default|SetProfPupVFXs _returnvalue_2)
      (elif (Utilities|Equal(Object) (Variables|Default|GetNewPart) "/BRCosmetics/Characters/CharacterParts/FaceAccessories/CP_M_LRG_ProfessorPup_Bad.CP_M_LRG_ProfessorPup_Bad")
        ;[1 1 1] Bad
        (Utilities|IsValid _profpupvfxs
          (:Is%20Valid
            (Components|Activation|Deactivate _profpupvfxs))
          (:Is%20Not%20Valid
            _))
        (bind _returnvalue_3 (Niagara|SpawnSystemAttached "/BRCosmetics/Effects/Fort_Effects/Effects/Characters/Athena_Parts/ProfessorPup/NS_ProfessorPup_Bad.NS_ProfessorPup_Bad" (Customization|GetPartSkeletalMeshComponent "Body")))
        (Variables|Default|SetProfPupVFXs _returnvalue_3)
        (else
          (Utilities|IsValid _profpupvfxs_2
            (:Is%20Valid
              (Components|Activation|Deactivate _profpupvfxs_2))))))))

(fn RacerZeroSetup
  (bind _racer_zero_head_vfx (Variables|Default|GetRacer%20Zero%20Head%20VFX))
  (bind _racer_zero_head_vfx_2 (Variables|Default|GetRacer%20Zero%20Head%20VFX))
  (bind _racer_zero_body2vfx (Variables|Default|GetRacer%20Zero%20Body2VFX))
  (bind _racerzerobody1vfx (Variables|Default|GetRacerZeroBody1VFX))
  (bind _racerzerobody1vfx_2 (Variables|Default|GetRacerZeroBody1VFX))
  (bind _racer_zero_body2vfx_2 (Variables|Default|GetRacer%20Zero%20Body2VFX))
  (if (Utilities|Equal(Object) (Variables|Default|GetNewPart) "/BRCosmetics/Athena/Heroes/Meshes/Bodies/CP_Athena_Body_M_RacerZero_T3.CP_Athena_Body_M_RacerZero_T3")
    (Utilities|IsValid _racerzerobody1vfx
      (:Is%20Valid
        (Components|Activation|Deactivate _racerzerobody1vfx))
      (:Is%20Not%20Valid
        _))
    (bind _returnvalue (Effects|SpawnEmitterAttached :EmitterTemplate "/BRCosmetics/Characters/Player/Male/Medium/Bodies/M_MED_Racer_Zero_Mask/FX/P_MED_Racer_Zero_Mask_ParticleSmoke_Body.P_MED_Racer_Zero_Mask_ParticleSmoke_Body" :AttachToComponent (Customization|GetPartSkeletalMeshComponent "Body") :AttachPointName "root" :Scale "1.000000,1.000000,1.000000" :LocationType "SnapToTargetIncludingScale" :bAutoDestroy true :PoolingMethod "None" :bAutoActivate true))
    (Variables|Default|SetRacerZeroBody1VFX _returnvalue)
    (Utilities|IsValid _racer_zero_body2vfx
      (:Is%20Valid
        (Components|Activation|Deactivate _racer_zero_body2vfx))
      (:Is%20Not%20Valid
        _))
    (bind _returnvalue_2 (Effects|SpawnEmitterAttached :EmitterTemplate "/BRCosmetics/Characters/Player/Male/Medium/Bodies/M_MED_Racer_Zero_Mask/FX/P_MED_Racer_Zero_Mask_ParticleSmoke_SpawnUnit.P_MED_Racer_Zero_Mask_ParticleSmoke_SpawnUnit" :AttachToComponent (Customization|GetPartSkeletalMeshComponent "Body") :AttachPointName "root" :Scale "1.000000,1.000000,1.000000" :LocationType "SnapToTargetIncludingScale" :bAutoDestroy true :PoolingMethod "None" :bAutoActivate true))
    (Variables|Default|SetRacer%20Zero%20Body2VFX _returnvalue_2)
    (Utilities|IsValid _racer_zero_head_vfx
      (:Is%20Valid
        (Components|Activation|Deactivate _racer_zero_head_vfx))
      (:Is%20Not%20Valid
        _))
    (bind _returnvalue_3 (Effects|SpawnEmitterAttached "/BRCosmetics/Characters/Player/Male/Medium/Bodies/M_MED_Racer_Zero_Mask/FX/P_MED_Racer_Zero_Mask_ParticleSmoke_Head.P_MED_Racer_Zero_Mask_ParticleSmoke_Head" (Customization|GetPartSkeletalMeshComponent "Head") "head"))
    (Variables|Default|SetRacer%20Zero%20Head%20VFX _returnvalue_3)
    (elif (or (Utilities|Equal(Object) (Variables|Default|GetNewPart) "/BRCosmetics/Characters/CharacterParts/Male/Medium/Heads/CP_Athena_Head_M_RacerZero.CP_Athena_Head_M_RacerZero") (Utilities|Equal(Object) (Variables|Default|GetNewPart) "/BRCosmetics/Characters/CharacterParts/FaceAccessories/CP_M_MED_RacerZero_T3.CP_M_MED_RacerZero_T3"))
      _
      (else
        (Utilities|IsValid _racerzerobody1vfx_2
          (:Is%20Valid
            (Components|Activation|Deactivate _racerzerobody1vfx_2))
          (:Is%20Not%20Valid
            _))
        (Utilities|IsValid _racer_zero_body2vfx_2
          (:Is%20Valid
            (Components|Activation|Deactivate _racer_zero_body2vfx_2))
          (:Is%20Not%20Valid
            _))
        (Utilities|IsValid _racer_zero_head_vfx_2
          (:Is%20Valid
            (Components|Activation|Deactivate _racer_zero_head_vfx_2)))))))

(fn ConstructionScript)
